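# Physics asset for the hair part of the female medium Street Fashion Garnet character, defining secondary motion for hair and a hat top.
# Capsule bodies form jointed chains rooted on a head capsule: center, left and right bangs, four three-bone hair strands, and a two-bone hat top.
# Constraints mostly use narrow swing and twist limits, often wider toward strand tips.
# The exception is the fourth strand's root, whose swing and twist are free.
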PhysicsAsset /BRCosmetics/Characters/Player/Female/Medium/Bodies/F_MED_Street_Fashion_Garnet/Meshes/Parts/F_MED_Garnet_Hair_Physics
bodies {
 dyn_bang_c_01 [{shapeType=Capsule, length=2.0077, radius=1.3377, center=<0.1037 0 0.83>, rotation=<rot 0 -89.9997 -82.8756>}]
 dyn_bang_c_02 [{shapeType=Capsule, length=2.9622, radius=1.5086, center=<2.4563 -0.3285 0.3621>, rotation=<rot -5.1437 -97.8935 105.0097>}]
 dyn_bang_l_01 [{shapeType=Capsule, length=2.2159, radius=2.7168, center=<1.8 0 0.35>, rotation=<rot 90 0 0>}]
 dyn_bang_l_02 [{shapeType=Capsule, length=2.2122, radius=1.3256, center=<2 0 0.35>, rotation=<rot -80 0 0>}]
 dyn_bang_r_01 [{shapeType=Capsule, length=2.9622, radius=1.5086, rotation=<rot 0 -89.9999 62.0374>}]
 dyn_bang_r_02 [{shapeType=Capsule, length=2.9622, radius=1.5086, center=<0.7637 0 -0.1452>, rotation=<rot 0 -89.9995 81.1414>}]
 dyn_bang_r_03 [{shapeType=Capsule, length=4.5, radius=1.6353, center=<-1.5266 0.2857 -0.5332>, rotation=<rot 90 0 0>}]
 dyn_hair1_01 [{shapeType=Capsule, length=2, radius=1.5, center=<0 0.6044 0>, rotation=<rot 90 0 0>}]
 dyn_hair1_02 [{shapeType=Capsule, length=2.4463, radius=1.8712, center=<0 0.6754 0>, rotation=<rot 90 0 0>}]
 dyn_hair1_03 [{shapeType=Capsule, length=1.75, radius=3.3712, center=<1 -3.5 0>, rotation=<rot 90 0 0>}]
 dyn_hair2_01 [{shapeType=Capsule, length=2, radius=1.5, center=<0 0.4527 0>, rotation=<rot 90 0 0>}]
 dyn_hair2_02 [{shapeType=Capsule, length=2.4463, radius=1.8712, center=<0 0.2392 0.6886>, rotation=<rot 90 0 0>}]
 dyn_hair2_03 [{shapeType=Capsule, length=2.4463, radius=1.8712, center=<2 0 0>, rotation=<rot 90 0 0>}]
 dyn_hair3_01 [{shapeType=Capsule, length=2, radius=1.5, center=<0 1.492 0>, rotation=<rot 90 0 0>}]
 dyn_hair3_02 [{shapeType=Capsule, length=2.4463, radius=1.8712, center=<0 0.4917 0>, rotation=<rot 90 0 0>}]
 dyn_hair3_03 [{shapeType=Capsule, length=3.4463, radius=1.8712, center=<0.5 0.0359 0>, rotation=<rot 90 0 0>}]
 dyn_hair4_01 [{shapeType=Capsule, length=0.505, radius=0.505, rotation=<rot 0 -90 0>}]
 dyn_hair4_02 [{shapeType=Capsule, length=2.4463, radius=1.8712, center=<1 0.4917 0>, rotation=<rot 90 0 0>}]
 dyn_hair4_03 [{shapeType=Capsule, length=1.9463, radius=1.6212, center=<1.5 0.4917 0>, rotation=<rot 90 0 0>}]
 hat_top_01 [{shapeType=Capsule, length=2, radius=2.75, center=<0.25 0 0>, rotation=<rot 90 0 0>}]
 hat_top_02 [{shapeType=Capsule, length=2, radius=1.75, center=<0.55 0 0>, rotation=<rot 90 0 0>}]
 head [{shapeType=Capsule, length=4.5, radius=6.2388, rotation=<rot 90 0 0>}]
}
constraints [
 {bone1Name=dyn_bang_c_01, bone2Name=head, swing1LimitDegrees=5, swing2LimitDegrees=45, twistLimitDegrees=5}
 {bone1Name=dyn_bang_c_02, bone2Name=dyn_bang_c_01, swing1LimitDegrees=45, swing2LimitDegrees=45, twistLimitDegrees=5}
 {bone1Name=dyn_bang_l_01, bone2Name=head, swing1LimitDegrees=10, swing2LimitDegrees=10, twistLimitDegrees=5}
 {bone1Name=dyn_bang_l_02, bone2Name=dyn_bang_l_01, swing1LimitDegrees=20, swing2LimitDegrees=20, twistLimitDegrees=5}
 {bone1Name=dyn_bang_r_01, bone2Name=head, swing1LimitDegrees=0, swing2LimitDegrees=10, twistLimitDegrees=5}
 {bone1Name=dyn_bang_r_02, bone2Name=dyn_bang_r_01, swing1LimitDegrees=10, swing2LimitDegrees=10, twistLimitDegrees=5}
 {bone1Name=dyn_bang_r_03, bone2Name=dyn_bang_r_02, swing1LimitDegrees=5, swing2LimitDegrees=10, twistLimitDegrees=5}
 {bone1Name=dyn_hair1_01, bone2Name=head, swing1LimitDegrees=5, swing2LimitDegrees=5, twistLimitDegrees=5}
 {bone1Name=dyn_hair1_02, bone2Name=dyn_hair1_01, swing1LimitDegrees=5, swing2LimitDegrees=5, twistLimitDegrees=5}
 {bone1Name=dyn_hair1_03, bone2Name=dyn_hair1_02, swing1LimitDegrees=5, swing2LimitDegrees=5, twistLimitDegrees=5}
 {bone1Name=dyn_hair2_01, bone2Name=head, swing1LimitDegrees=5, swing2LimitDegrees=5, twistLimitDegrees=5}
 {bone1Name=dyn_hair2_02, bone2Name=dyn_hair2_01, swing1LimitDegrees=5, swing2LimitDegrees=5, twistLimitDegrees=5}
 {bone1Name=dyn_hair2_03, bone2Name=dyn_hair2_02, swing1LimitDegrees=25, swing2LimitDegrees=25, twistLimitDegrees=5}
 {bone1Name=dyn_hair3_01, bone2Name=head, swing1LimitDegrees=5, swing2LimitDegrees=5, twistLimitDegrees=5}
 {bone1Name=dyn_hair3_02, bone2Name=dyn_hair3_01, swing1LimitDegrees=5, swing2LimitDegrees=5, twistLimitDegrees=5}
 {bone1Name=dyn_hair3_03, bone2Name=dyn_hair3_02, swing1LimitDegrees=25, swing2LimitDegrees=33.5714, twistLimitDegrees=5}
 {
  bone1Name=dyn_hair4_01
  bone2Name=head
  swing1LimitDegrees=45
  swing1Motion=Free
  swing2LimitDegrees=45
  swing2Motion=Free
  twistLimitDegrees=45
  twistMotion=Free
 }
 {bone1Name=dyn_hair4_02, bone2Name=dyn_hair4_01, swing1LimitDegrees=10, swing2LimitDegrees=20, twistLimitDegrees=5}
 {bone1Name=dyn_hair4_03, bone2Name=dyn_hair4_02, swing1LimitDegrees=25, swing2LimitDegrees=38.7143, twistLimitDegrees=5}
 {bone1Name=hat_top_01, bone2Name=head, swing1LimitDegrees=5, swing2LimitDegrees=5, twistLimitDegrees=5}
 {bone1Name=hat_top_02, bone2Name=hat_top_01, swing1LimitDegrees=8, swing2LimitDegrees=8, twistLimitDegrees=5}
]
